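# Material function that supplies terrain variation masks for the mobile Hermes landscape materials.
# It converts camera-relative world position to periodic world space with an 8192-unit tile, keeping coordinates small for precision in large worlds.
# Dividing by the tile size and keeping XY gives top-down UVs for one sample of the Hermes mobile noise texture.
# All four RGBA channels go out as a single VariationMasks Mobile output.
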MaterialFunction /BRCoreMaterials/Hermes/Terrain/Landscape/Materials/MaterialFunctions/MF_Hermes_VariationMasks_Mobile
graph {nodes [
  n1 = Constant() {r=8192}
  n2 = Divide(
   A=TransformPosition!(None=WorldPosition() {worldPositionShaderOffset=WPT_CameraRelative}.XYZ, Periodic World Tile Size=@n1) {transformSourceType=TRANSFORMPOSSOURCE_TranslatedWorld, transformType=TRANSFORMPOSSOURCE_PeriodicWorld}
   B=@n1
  )
  FunctionOutput(None=TextureSample(UVs=ComponentMask(None=@n2) {g=true, r=true}) {texture=/Game/Environments/Helios/Terrain/TexturesMobile/T_Hermes_Mobile_Noise}.RGBA) {outputName=VariationMasks Mobile}
 ]}
